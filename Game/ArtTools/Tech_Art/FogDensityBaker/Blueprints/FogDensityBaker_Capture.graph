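# FogDensityBaker_Capture is an Editor Utility Blueprint in the FogDensityBaker tech-art tool folder, with almost no working logic.
# Its construction script holds an unconnected fragment that builds a transform from the actor's location lowered 30,000 units in Z, keeping its rotation.
# That transform feeds nothing, the BeginPlay, overlap and Tick events are empty, and no variables are declared.
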EditorUtilityBlueprint /Game/ArtTools/Tech_Art/FogDensityBaker/Blueprints/FogDensityBaker_Capture
graphs [
 EventGraph {nodes [
   Event BeginPlay()
   Event ActorBeginOverlap()
   Event Tick!() -> (OutputDelegate, then, DeltaSeconds:float)
  ]}
 UserConstructionScript {nodes [
   n1 = Get Actor Transform!(self) -> (ReturnValue_Location:Vector, ReturnValue_Rotation:Rotator, ReturnValue_Scale:Vector)
   MakeTransform(Location=vector - vector(A=@n1, B="0,0,30000"), Rotation=@n1.ReturnValue_Rotation)
   Construction Script()
  ]}
]
variables []
