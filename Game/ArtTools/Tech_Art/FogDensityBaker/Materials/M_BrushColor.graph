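# Masked, Default Lit brush color material in the Fog Density Baker tech-art tool, enabled for Nanite.
# It builds a vector with one minus the Density scalar parameter in X and a constant in Y.
# The graph also holds an object radius scaled down by a constant, an orthographic-view check and loose constants, none feeding that vector.
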Material /Game/ArtTools/Tech_Art/FogDensityBaker/Materials/M_BrushColor
graph {
 properties {bUsedWithNanite=true, blendMode=BLEND_Masked, pixelDepthOffsetMode=PDOM_Legacy, usedShadingModels=Default Lit}
 nodes [
  Constant()
  Constant()
  Divide(A=ObjectRadius()) {constB=1000}
  IsOrthographic()
  MaterialFunctionCall(X=OneMinus(None=ScalarParameter() {defaultValue=0.5, parameterName=Density, sliderMax=1}), Y=Constant() {r=0.5}) {materialFunction=&MakeFloat3}
 ]
}
