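# Unlit, two-sided, masked wireframe material in the Fog Density Baker tech-art tool, named for drawing the fog density mesh.
# A Color vector parameter, olive green by default, supplies a flat tint.
# An inverted, saturated IsOrthographic test yields one in perspective and zero in orthographic views, apparently as a mask hiding the wireframe in orthographic views.
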Material /Game/ArtTools/Tech_Art/FogDensityBaker/Materials/M_FogDensityMeshWireframe
graph {
 properties {blendMode=BLEND_Masked, pixelDepthOffsetMode=PDOM_Legacy, shadingModel=MSM_Unlit, twoSided=true, usedShadingModels=Unlit, wireframe=true}
 nodes [
  OneMinus(None=Saturate(None=IsOrthographic()))
  VectorParameter() {parameterName=Color, defaultValue=<rgba 0.3281 0.5469 0 1>}
 ]
}
