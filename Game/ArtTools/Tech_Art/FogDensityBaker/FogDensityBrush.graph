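# FogDensityBrush is an editor-only utility actor that shows a fog density value on a configurable mesh, apparently for the fog density baker tool.
# In the editor, its construction script applies Mesh to the Density Mesh component and feeds Density into a dynamic M_BrushColor material.
# It adds a second, slightly Z-shrunk copy of the mesh with a wireframe material.
# Outside the editor, the construction script destroys the actor.
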EditorUtilityBlueprint /Game/ArtTools/Tech_Art/FogDensityBaker/FogDensityBrush
graphs [
 EventGraph {nodes [
   Event BeginPlay()
   Event ActorBeginOverlap()
   Event Tick!() -> (OutputDelegate, then, DeltaSeconds:float)
  ]}
 UserConstructionScript {nodes [
   Construction Script()
   > n1 = Branch(Condition=InEditor!(WorldContextObject:Object Reference) -> (ReturnValue:bool))
   > SetStaticMesh(self=Get Density Mesh() -> (Density Mesh:Static Mesh Component Object Reference), NewMesh=Get Mesh!() -> (Mesh:Static Mesh Object Reference))
   > n2 = CreateDynamicMaterialInstance!(
    execute
    self:Primitive Component Object Reference=Get Density Mesh()
    ElementIndex:int=0
    SourceMaterial:Material Interface Object Reference=/Game/ArtTools/Tech_Art/FogDensityBaker/Materials/M_BrushColor
    OptionalName:Name=None
   ) -> (then, ReturnValue:Material Instance Dynamic Object Reference) {is=title K2Node_CallFunction}
   > SetScalarParameterValue(self=@n2, ParameterName=Density, Value=Get Density() -> (Density:float))
   > n3 = Add Static Mesh Component!(
    execute:Exec
    self:Actor Object Reference
    bManualAttachment:bool=false
    RelativeTransform:Transform=MakeTransform(Rotation="0, 0, 0", Scale="1,1,0.99")
    ComponentTemplateContext:Object Reference
   ) -> (then:Exec, ReturnValue:Static Mesh Component Object Reference) {is=title K2Node_AddComponent}
   > SetStaticMesh(self=@n3, NewMesh=Get Mesh())
   > SetMaterial(self=@n3, ElementIndex=0, Material=/Game/ArtTools/Tech_Art/FogDensityBaker/Materials/M_FogDensityMeshWireframe)
   Destroy Actor(execute=@n1.else)
  ]}
]
variables [
 Density {type=number, default=0.25}
 Mesh {type=StaticMesh, default=/Engine/BasicShapes/Cylinder}
]
