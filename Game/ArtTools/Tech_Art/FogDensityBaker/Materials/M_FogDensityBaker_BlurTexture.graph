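# Unlit material in the FogDensityBaker tech-art tool that blurs the RT_FogDensityBaker render target.
# SpiralBlur-Texture samples the target with a distance set by the Blur parameter and temporally dithered step counts, feeding a RelativeBlur lerp and a red-channel inversion.
# A separate EdgeDistanceField custom node holds a radial edge search, but an early return makes it output only the raw sample.
# The graph shows no link to a material output.
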Material /Game/ArtTools/Tech_Art/FogDensityBaker/Materials/M_FogDensityBaker_BlurTexture
graph {
 properties {pixelDepthOffsetMode=PDOM_Legacy, shadingModel=MSM_Unlit, usedShadingModels=Unlit}
 nodes [
  MaterialExpressionLinearInterpolate_6 = LinearInterpolate(A=ScalarParameter() {defaultValue=29, parameterName=Iterations}, B=ScalarParameter() {defaultValue=29, parameterName=Iterations})
  MaterialExpressionConstant_3 = Constant() {r=0.1}
  MaterialExpressionScalarParameter_4 = ScalarParameter() {defaultValue=0.03, parameterName=Blur}
  MaterialExpressionScalarParameter_9 = ScalarParameter() {defaultValue=31}
  MaterialExpressionConstant_4 = Constant() {r=0.01}
  MaterialExpressionTextureObject_2 = TextureObject() {samplerType=SAMPLERTYPE_LinearColor, texture=/Game/ArtTools/Tech_Art/FogDensityBaker/Textures/RT_FogDensityBaker}
  MaterialExpressionTextureCoordinate_0 = TextureCoordinate()
  Custom!(
   Iterations=@MaterialExpressionLinearInterpolate_6
   Threshold=@MaterialExpressionConstant_3
   step=@MaterialExpressionScalarParameter_4
   radSamp=@MaterialExpressionScalarParameter_9
   phase=@MaterialExpressionConstant_4
   Tex=@MaterialExpressionTextureObject_2
   UV=@MaterialExpressionTextureCoordinate_0
  ) {
   code="float2 NewUV = UV;\r\n\r\nfloat initD = Texture2DSample(Tex,TexSampler,UV);\r\nfloat d;\r\n\r\nreturn initD;\r\n\r\nfloat CurDistance=0;\r\nfloat2 CurOffset;\r\nfloat TwoPi = 6.283185;\r\nint RadialSamples = radSamp;\r\n\r\nfloat phs = phase * 0.5;\r\n\r\nstep = step/Iterations;\r\n\r\nfor(int i = 0; i < Iterations; i++)\r\n{\r\n\tfor(int j = 0; j < RadialSamples ; j++)\r\n\t{\r\n\t\tCurOffset.x = cos( (TwoPi*j+phs) / RadialSamples) * (step * i);\r\n\t\tCurOffset.y = sin( (TwoPi*j+phs) / RadialSamples) * (step * i);\r\n\t\tNewUV = saturate(UV + CurOffset);\r\n\t\td = max(Texture2DSample(Tex,TexSampler,NewUV), d);\r\n\t\t\r\n\t\tif( distance(d, initD) > Threshold)\r\n\t\t{\r\n\t\t\treturn 1- (float(i)/Iterations);\r\n\t\t}\r\n\r\n\t}\r\n}\r\n\r\nreturn 0;"
   description=EdgeDistanceField
   outputType=CMOT_Float1
   inputs [
    {
     inputName=Iterations
     input {expression=&self:MaterialExpressionLinearInterpolate_6}
    }
    {
     inputName=Threshold
     input {expression=&self:MaterialExpressionConstant_3}
    }
    {
     inputName=step
     input {expression=&self:MaterialExpressionScalarParameter_4}
    }
    {
     inputName=radSamp
     input {expression=&self:MaterialExpressionScalarParameter_9}
    }
    {
     inputName=phase
     input {expression=&self:MaterialExpressionConstant_4}
    }
    {
     inputName=Tex
     input {expression=&self:MaterialExpressionTextureObject_2}
    }
    {
     inputName=UV
     input {expression=&self:MaterialExpressionTextureCoordinate_0}
    }
   ]
  }
  n1 = TextureObject() {samplerType=SAMPLERTYPE_LinearColor, texture=&RT_FogDensityBaker}
  n2 = Constant() {r=12}
  n3 = Constant() {r=13}
  n4 = MaterialFunctionCall() {materialFunction=&DitherTemporalAA}
  n5 = MaterialFunctionCall!(
   TextureObject=@n1
   UV
   Distance=LinearInterpolate(Alpha=ScalarParameter() {defaultValue=0.03, parameterName=Blur, sliderMax=1}) {constA=0.001, constB=0.15}
   Distance Steps=LinearInterpolate(A=@n2, B=@n3, Alpha=@n4.Result)
   Radial Steps=LinearInterpolate(A=@n3, B=@n2, Alpha=@n4.Result)
   Radial Offset
   Kernel Power
  ) {materialFunction=/Engine/Functions/Engine_MaterialFunctions01/Texturing/SpiralBlur-Texture}
  LinearInterpolate(A=@n5.Result, Alpha=ScalarParameter() {parameterName=RelativeBlur, sliderMax=1})
  MaterialFunctionCall() {materialFunction=&DitherTemporalAA}
  OneMinus(None=ComponentMask(None=@n5.Result) {r=true})
  TextureSample(Tex=@n1)
 ]
}
