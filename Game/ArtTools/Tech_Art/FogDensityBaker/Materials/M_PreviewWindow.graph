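# UI-domain material for the preview window of the FogDensityBaker tech-art tool.
# It holds a single linear-color texture sample that references the blurred fog density render target, RT_FogDensityBakerBlured.
# The graph contains no parameters or other processing nodes.
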Material /Game/ArtTools/Tech_Art/FogDensityBaker/Materials/M_PreviewWindow
graph {
 properties {materialDomain=MD_UI, pixelDepthOffsetMode=PDOM_Legacy, usedShadingModels=Default Lit}
 nodes [TextureSample() {samplerType=SAMPLERTYPE_LinearColor, texture=/Game/ArtTools/Tech_Art/FogDensityBaker/Textures/RT_FogDensityBakerBlured}]
}
