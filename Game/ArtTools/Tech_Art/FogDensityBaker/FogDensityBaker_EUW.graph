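# Editor utility widget for painting and baking a fog density texture over the bounds of Jungle day sequence volumes.
# A spawned capture actor renders only FogDensityBrush actors orthographically over those bounds; each tick a blur material redraws the render target, which Save writes to T_FogDensity.
# New Brush spawns a brush at the camera trace.
# SelectActorFromTexture maps mouse position to world XY, sphere-tracing vertically to select brushes and align the camera.
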EditorUtilityWidgetBlueprint /Game/ArtTools/Tech_Art/FogDensityBaker/FogDensityBaker_EUW
graphs [
 EventGraph {nodes [
   "On Clicked (Button_130)"() {is=title K2Node_ComponentBoundEvent}
   > ConvertRenderTargetToTexture2DEditorOnly(
    RenderTarget=Get RenderTargetFinal() -> (RenderTargetFinal:Texture Render Target 2D Object Reference)
    Texture=&/T_FogDensity
   )
   "On Clicked (Button)"() {is=title K2Node_ComponentBoundEvent}
   n1 = GetLevelViewportCameraInfo!(self:Unreal Editor Subsystem Object Reference=Get UnrealEditorSubsystem()) -> (CameraLocation:Vector, CameraRotation:Rotator, ReturnValue:bool)
   n2 = vector + vector(A=@n1, B=vector * vector(A=GetForwardVector!(InRot:Rotator=@n1.CameraRotation) -> (ReturnValue:Vector), B=40000) -> (ReturnValue=~))
   > n3 = Line Trace By Channel!(
    execute
    WorldContextObject:Object Reference
    Start:Vector=@n1
    End:Vector=@n2
    TraceChannel:ETraceTypeQuery Enum
    bTraceComplex:bool=false
    ActorsToIgnore:Array of Actor Object References
    DrawDebugType:EDrawDebugTrace Enum
    bIgnoreSelf:bool
    TraceColor:Linear Color Structure
    TraceHitColor:Linear Color Structure
    DrawTime:float
   ) -> (then, OutHit:Hit Result Structure, ReturnValue:bool)
   n4 = BreakHitResult!(Hit:"Hit Result Structure (by ref)"=@n3) -> (
    bBlockingHit:bool
    bInitialOverlap:bool
    Time:float
    Distance:float
    Location:Vector
    ImpactPoint:Vector
    Normal:Vector
    ImpactNormal:Vector
    PhysMat:Physical Material Object Reference
    HitActor:Actor Object Reference
    HitComponent:Primitive Component Object Reference
    HitBoneName:Name
    BoneName:Name
    HitItem:int
    ElementIndex:int
    FaceIndex:int
    TraceStart:Vector
    TraceEnd:Vector
   )
   > SpawnActor Fog Density Brush!(
    execute:Exec
    WorldContextObject:Object Reference
    Class:Actor Class Reference=&/FogDensityBrush.FogDensityBrush_C
    SpawnTransform:Transform=MakeTransform(Location=SelectVector(A=@n4.Location, B=@n2, bPickA=@n3.ReturnValue), Rotation="0, 0, 0", Scale="50,50,50")
    CollisionHandlingOverride:ESpawnActorCollisionHandlingMethod Enum=Undefined
    TransformScaleMethod:ESpawnActorScaleMethod Enum=MultiplyWithRoot
    Owner:Actor Object Reference
    Instigator:Pawn Object Reference
   ) -> (then:Exec, ReturnValue:Fog Density Brush Object Reference) {is=title K2Node_SpawnActorFromClass}
   Event PreConstruct()
   > SetObject(self=Get DetailsView_39() -> (DetailsView_39:Details View Object Reference), NewObject=Self-Reference())
   Event Construct()
   > n5 = GetAllActorsOfClass!(
    execute
    WorldContextObject:Object Reference
    ActorClass:Actor Class Reference=/Game/ArtTools/Tech_Art/FogDensityBaker/Blueprints/FogDensityBaker_Capture.FogDensityBaker_Capture_C
   ) -> (then, OutActors:Array of Fog Density Baker Capture Object References)
   > n6 = For Each Loop(Array:Array of Fog Density Baker Capture Object References=@n5) -> (Array Element:Fog Density Baker Capture Object Reference)
   > Destroy Actor(self=@n6)
   n7 = SpawnActorFromClass!(
    execute=@n6.Completed
    ActorClass:Actor Class Reference=&FogDensityBaker_Capture_C
    Location:Vector=Get BoundsCenter() -> (BoundsCenter:Vector)
    Rotation:Rotator
    bTransient:bool=true
   ) -> (then, ReturnValue:Fog Density Baker Capture Object Reference)
   > Set CaptureActor(CaptureActor:Fog Density Baker Capture Object Reference=@n7) -> (Output_Get:Fog Density Baker Capture Object Reference)
   > n8 = GetAllActorsOfClass(ActorClass=/Game/TimeOfDay/TODM_Sequence/Blueprint/Volume/Ch4s3/B_DaySequenceVolume_Jungle.B_DaySequenceVolume_Jungle_C) -> (OutActors:Array of B Day Sequence Volume Jungle Object References)
   > Set Actors(Actors:Array of Actor Object References=@n8) -> (Output_Get:Array of Actor Object References)
   n9 = Event Tick()
   > n10 = Sequence!(execute) -> (then_0, then_1, then_2:Exec, then_3:Exec, then_4:Exec, then_5:Exec)
   > n11 = GetSelectedActors()
   > Branch(Condition=Is Not Empty!(TargetArray:Array of Actor Object References=@n11) -> (ReturnValue:bool))
   > n12 = Cast To FogDensityBrush!(execute, Object="Get (a copy)"(Array:Array of Actor Object References=@n11) -> (Output:Actor Object Reference)) -> (then, CastFailed, AsFog Density Brush:Fog Density Brush Object Reference) {is=title K2Node_DynamicCast}
   > SetObject(self=Get SinglePropertyView_68() -> (SinglePropertyView_68:Single Property View Object Reference), NewObject=@n12)
   > SetObject(self=Get SinglePropertyView() -> (SinglePropertyView:Single Property View Object Reference), NewObject=Self-Reference())
   n13 = For Each Loop(Exec=@n10.then_1, Array:Array of Actor Object References=Get Actors() -> (Actors:Array of Actor Object References)) -> (Array Element:Actor Object Reference)
   n14 = Get Actor Bounds(self=@n13, bOnlyCollidingComponents=false)
   n15 = vector - vector!(A=@n14, B=@n14.BoxExtent) -> (ReturnValue_X:double, ReturnValue_Y:double, ReturnValue_Z:double)
   n16 = Get!() -> (BoundsMin_X:double, BoundsMin_Y:double, BoundsMin_Z:double)
   > Set BoundsMin(BoundsMin:Vector=MakeVector(
     X="Min (Float)"(A:double=@n15, B:double=@n16) -> (ReturnValue:double) {is=title K2Node_CommutativeAssociativeBinaryOperator}
     Y="Min (Float)"(A=@n15.ReturnValue_Y, B=@n16.BoundsMin_Y)
     Z="Min (Float)"(A=@n15.ReturnValue_Z, B=@n16.BoundsMin_Z)
    )) -> (Output_Get:Vector)
   n17 = vector + vector!(A=@n14, B=@n14.BoxExtent) -> (ReturnValue_X:double, ReturnValue_Y:double, ReturnValue_Z:double)
   n18 = Get!() -> (BoundsMax_X:double, BoundsMax_Y:double, BoundsMax_Z:double)
   > Set BoundsMax(BoundsMax:Vector=MakeVector(
     X="Max (Float)"(A=@n17, B=@n18)
     Y="Max (Float)"(A=@n17.ReturnValue_Y, B=@n18.BoundsMax_Y)
     Z="Max (Float)"(A=@n17.ReturnValue_Z, B=@n18.BoundsMax_Z)
    )) -> (Output_Get:Vector)
   n19 = Get CaptureActor() -> (CaptureActor:Fog Density Baker Capture Object Reference)
   Set Actor Location!(execute=@n10.then_3, self:Actor Object Reference=@n19, NewLocation:Vector=Get BoundsCenter(), bSweep:bool=false, bTeleport:bool=false) -> (then, SweepHitResult:Hit Result Structure, ReturnValue:bool) {is=title K2Node_CallFunction}
   > Add Actor World Offset!(execute, self:Actor Object Reference=@n19, DeltaLocation:Vector="0,0,50000", bSweep:bool=false, bTeleport:bool=true) -> (then, SweepHitResult:Hit Result Structure)
   n20 = Get BoundsExtents() -> (BoundsExtents:Vector)
   > Set OrthoWidth!(
    execute
    OrthoWidth:float=float * float(A=GetMaxElement!(A:Vector=@n20) -> (ReturnValue:double), B=2) {is=title K2Node_PromotableOperator}
    self:Scene Capture Component 2D Object Reference=Get SceneCaptureComponent2D!(self:Fog Density Baker Capture Object Reference=Get CaptureActor()) -> (SceneCaptureComponent2D:Scene Capture Component 2D Object Reference)
   ) -> (then, Output_Get:float)
   > DrawDebugBox!(
    execute
    WorldContextObject:Object Reference
    Center:Vector=Get BoundsCenter()
    Extent:Vector=Get BoundsExtents()
    LineColor:Linear Color Structure=(R=0,G=0.3922,B=1,A=1)
    Rotation:Rotator
    Duration:float=0
    Thickness:float=250
    DepthPriority:EDrawDebugSceneDepthPriorityGroup Enum=World
   ) -> (then)
   n21 = GetMaxElement(A=Get BoundsExtents())
   > DrawDebugBox!(
    execute
    WorldContextObject
    Center=Get BoundsCenter()
    Extent_X:double=@n21
    Extent_Y:double=@n21
    Extent_Z:double=Get!() -> (BoundsExtents_X:double, BoundsExtents_Y:double, BoundsExtents_Z:double).BoundsExtents_Z
    LineColor=(R=1,G=0.1947,B=0,A=1)
    Rotation
    Duration=0
    Thickness=1000
    DepthPriority=World
   ) -> (then)
   > n22 = CreateDynamicMaterialInstance(Parent=/Game/ArtTools/Tech_Art/FogDensityBaker/Materials/M_FogDensityBaker_BlurTexture, OptionalName=None, CreationFlags=None)
   > SetScalarParameterValue(self=@n22, ParameterName=Blur, Value=Get Blur() -> (Blur:float))
   > DrawMaterialToRenderTarget!(
    execute
    WorldContextObject:Object Reference
    TextureRenderTarget:Texture Render Target 2D Object Reference=/Game/ArtTools/Tech_Art/FogDensityBaker/Textures/RT_FogDensityBakerBlured
    Material:Material Interface Object Reference=@n22
   ) -> (then)
   Delay!(execute=@n10.then_4, WorldContextObject:Object Reference, Duration=0.5) -> (then)
   > n23 = GetAllActorsOfClass(ActorClass=&FogDensityBrush_C) -> (OutActors:Array of Fog Density Brush Object References)
   > Set ShowOnlyActors!(
    execute
    ShowOnlyActors:Array of Actor Object References=@n23
    self:Scene Capture Component Object Reference=Get SceneCaptureComponent2D(self=Get CaptureActor())
   ) -> (then, Output_Get:Array of Actor Object References)
   n24 = GetAbsoluteSize!(Geometry:"Geometry Structure (by ref)"=@n9.MyGeometry) -> (ReturnValue_X:double, ReturnValue_Y:double)
   n25 = "Min (Float)"(A=float - float(A=@n24, B=365), B=float - float(A=@n24.ReturnValue_Y, B=5))
   SetDesiredSizeOverride!(
    execute=@n10.then_5
    self:Image Object Reference=Get Image_0() -> (Image_0:Image Object Reference)
    DesiredSize:Vector 2D Structure=MakeVector2D(X=@n25, Y=@n25)
   ) -> (then)
   n26 = GetMaxElement(A=Get BoundsExtents())
   n27 = Get BoundsCenter()
   n28 = vector - vector(A=@n27, B:double=@n26)
   n29 = vector + vector(A=@n27, B:double=@n26)
   > Set Bounds(Bounds:Box 2D Structure=MakeBox2D!(Min_X:double=@n28, Min_Y:double=@n28.ReturnValue_Y, Max_X:double=@n29, Max_Y:double=@n29.ReturnValue_Y) -> (ReturnValue:Box 2D Structure)) -> (Output_Get:Box 2D Structure)
   n30 = Get BoundsMax() -> (BoundsMax:Vector)
   n31 = Get BoundsMin() -> (BoundsMin:Vector)
   Set BoundsCenter(
    execute=@n10.then_2
    BoundsCenter:Vector=vector / vector(A:Vector=vector + vector!(A=@n31, B:Vector=@n30) -> (ReturnValue:Vector), B:Vector="2,2,2") -> (ReturnValue:Vector)
   ) -> (Output_Get:Vector)
   n32 = vector - vector!(A=@n31, B:Vector=@n30) -> (ReturnValue:Vector)
   > Set BoundsExtents(BoundsExtents:Vector=Vector_GetAbs!(A:Vector=vector / vector(A=@n32, B="2,2,2")) -> (ReturnValue:Vector)) -> (Output_Get:Vector)
   Event Destruct()
   > Destroy Actor(self=Get CaptureActor())
   CreateNewTexture2DAsset!(
    execute
    FromTexture:Texture 2D Object Reference=Get From Texture() -> (From Texture:Texture 2D Object Reference)
    AssetPathAndName:String=&/T_FogDensity2
    Options:Geometry Script Create New Texture 2DAsset Options Structure=Make Geometry Script Create New Texture 2DAsset Options!(bOverwriteIfExists:bool=False) -> (GeometryScriptCreateNewTexture2DAssetOptions:Geometry Script Create New Texture 2DAsset Options Structure) {is=title K2Node_MakeStruct}
   ) -> (Failure:Exec, Success:Exec, ReturnValue:Texture 2D Object Reference)
   Append!(
    A=Content/ArtTools/Tech_Art/FogDensityBaker/T_FogDensity2
    B=T_FogDensity2
    C:String=D:/F/F2/FortniteGame/Content/ArtTools/Tech_Art/FogDensityBaker/T_FogDensity2
   ) -> (ReturnValue)
   Get EditorAssetSubsystem() -> (ReturnValue:Editor Asset Subsystem Object Reference) {is=title K2Node_GetEditorSubsystem}
  ]}
 SelectActorFromTexture {nodes [
   n1 = SelectActorFromTexture!() -> (then, MyGeometry:Geometry Structure, MouseEvent:"Pointer Event Structure (by ref)") {is=title K2Node_FunctionEntry}
   n2 = GetLocalTopLeft!(Geometry:"Geometry Structure (by ref)"=@n1) -> (ReturnValue:Vector 2D Structure)
   n3 = vector2d / vector2d!(
    A:Vector 2D Structure=vector2d - vector2d!(
     A:Vector 2D Structure=Get Screen Space Position!(Input:"Pointer Event Structure (by ref)"=@n1.MouseEvent) -> (ReturnValue:Vector 2D Structure) {is=title K2Node_CallFunction}
     B:Vector 2D Structure="Local to Absolute (Coordinates)"!(Geometry:"Geometry Structure (by ref)"=@n1, LocalCoordinate:Vector 2D Structure=@n2) -> (ReturnValue:Vector 2D Structure)
     C:Vector 2D Structure=(X=-360,Y=0)
    ) -> (ReturnValue:Vector 2D Structure) {is=title K2Node_PromotableOperator}
    B:Vector 2D Structure=GetAbsoluteSize!(Geometry=@n1) -> (ReturnValue:Vector 2D Structure)
   ) -> (ReturnValue_X:double, ReturnValue_Y:double) {is=title K2Node_PromotableOperator}
   n4 = GetMaxElement(A=Get BoundsExtents())
   n5 = Get BoundsCenter()
   n6 = vector + vector!(A=@n5, B:double=@n4) -> (ReturnValue_X:double, ReturnValue_Y:double, ReturnValue_Z:double)
   n7 = vector - vector!(A=@n5, B:double=@n4) -> (ReturnValue_X:double, ReturnValue_Y:double, ReturnValue_Z:double)
   n8 = Lerp(A=@n6.ReturnValue_Y, B=@n7.ReturnValue_Y, Alpha=@n3.ReturnValue_Y)
   n9 = Lerp(A=@n6, B=@n7, Alpha=@n3)
   > n10 = Multi Sphere Trace By Channel!(
    execute
    WorldContextObject:Object Reference
    Start:Vector=MakeVector(X=@n9, Y=@n8, Z=100000)
    End:Vector=MakeVector(X=@n9, Y=@n8, Z=-100000)
    Radius:float=1000
    TraceChannel:ETraceTypeQuery Enum
    bTraceComplex:bool=false
    ActorsToIgnore:Array of Actor Object References
    DrawDebugType:EDrawDebugTrace Enum
    bIgnoreSelf:bool=true
    TraceColor:Linear Color Structure=(R=1,G=0,B=0,A=1)
    TraceHitColor:Linear Color Structure=(R=0,G=1,B=0,A=1)
    DrawTime:float=5
   ) -> (then, OutHits:Array of Hit Result Structures, ReturnValue:bool)
   > n11 = For Each Loop(Array:Array of Hit Result Structures=@n10) -> (Array Element:Hit Result Structure)
   n12 = Get SelectActors() -> (SelectActors:Array of Fog Density Brush Object References)
   Branch(execute=@n11.Completed, Condition=Is Not Empty(TargetArray:Array of Fog Density Brush Object References=@n12))
   > SetSelectedLevelActors(self=Get EditorActorSubsystem(), ActorsToSelect=@n12)
   > ExecuteConsoleCommand(Command=CAMERA ALIGN ACTIVEVIEWPORTONLY)
   > Return Node!(execute, ReturnValue:Event Reply Structure=Make Event Reply!() -> (EventReply:Event Reply Structure) {is=title K2Node_MakeStruct})
   n13 = Cast To FogDensityBrush(execute=@n11.LoopBody, Object=BreakHitResult(Hit=@n11).HitActor)
   > Add Unique(TargetArray:Array of Fog Density Brush Object References=Get SelectActors(), NewItem:"Fog Density Brush Object Reference (by ref)"=@n13)
  ]}
]
variables [
 Actors {type="array<Actor>"}
 Blur {type=number, default=0.05}
 Bounds {type=Box2D, default {bIsValid=false, max=<0 0>, min=<0 0>}}
 BoundsCenter {type=Vector}
 BoundsExtents {type=Vector}
 BoundsMax {type=Vector, default=<-1000000 -1000000 -1000000>}
 BoundsMin {type=Vector, default=<10000000 10000000 10000000>}
 Brushes {type="array<Actor>"}
 CaptureActor {type=FogDensityBaker_Capture_C, default=null}
 From Texture {type=Texture2D, default=&T_BlackMask}
 RenderTargetFinal {type=TextureRenderTarget2D, default=&RT_FogDensityBakerBlured}
 Texture {type=TextureRenderTarget2D, default=null}
]
widgets {
 info {rootWidgetClass=&CanvasPanel, widgetCount=21}
 widgets [
  CanvasPanel_18 {widgetClassPath=&CanvasPanel, widget=CanvasPanel_18}
  StackBox_0 {
   widgetClassPath=&StackBox
   parent=CanvasPanel_18
   widget=StackBox_0
   properties {orientation=Orient_Vertical}
   slot {class=CanvasPanelSlot, properties {layoutData {alignment=<0 0>, anchors {maximum=<0 0>, minimum=<0 0>}, offsets=<ltrb 0 0 360 1076.5465>}}}
  }
  HorizontalBox_178 {
   widgetClassPath=&HorizontalBox
   parent=StackBox_0
   widget=HorizontalBox_178
   slot {class=StackBoxSlot, properties {padding=<ltrb 5 10 5 0>}}
  }
  VerticalBox_126 {
   widgetClassPath=&VerticalBox
   parent=HorizontalBox_178
   widget=VerticalBox_126
   slot {class=HorizontalBoxSlot, properties {size {sizeRule=Fill, value=1}}}
  }
  HorizontalBox_180 {
   widgetClassPath=&HorizontalBox
   parent=VerticalBox_126
   widget=HorizontalBox_180
   slot {class=VerticalBoxSlot, properties {padding=<ltrb 0 0 0 5>, size {sizeRule=Fill, value=1}}}
  }
  CommonTextBlock_104 {
   widgetClassPath=&CommonTextBlock
   parent=HorizontalBox_180
   widget=CommonTextBlock_104
   properties {
    style=/Game/ArtTools/Font/TextStyle_RobotoEngineLike.TextStyle_RobotoEngineLike_C
    text=Density
    colorAndOpacity {colorUseRule=UseColor_Specified, specifiedColor=<rgba 0.5 0.5 0.5 1>}
    font {
     bForceMonospaced=false
     bMaterialIsStencil=false
     fontMaterial=null
     fontObject=&Roboto
     letterSpacing=0
     monospacedWidth=1
     size=8
     skewAmount=0
     typefaceFontName=Regular
     outlineSettings {
      bApplyOutlineToDropShadows=false
      bMiteredCorners=false
      bSeparateFillAlpha=false
      outlineMaterial=null
      outlineSize=0
      outlineColor=<rgba 0 0 0 1>
     }
    }
    shadowOffset=<0 0>
   }
   slot {
    class=HorizontalBoxSlot
    properties {horizontalAlignment=HAlign_Left, verticalAlignment=VAlign_Center, padding=<ltrb 5 0 0 0>, size {sizeRule=Fill, value=0.15}}
   }
  }
  SinglePropertyView_68 {
   widgetClassPath=&SinglePropertyView
   parent=HorizontalBox_180
   widget=SinglePropertyView_68
   properties {nameOverride=" ", propertyName=Density}
   slot {class=HorizontalBoxSlot, properties {size {sizeRule=Fill, value=1}}}
  }
  HorizontalBox_181 {
   widgetClassPath=&HorizontalBox
   parent=VerticalBox_126
   widget=HorizontalBox_181
   slot {class=VerticalBoxSlot, properties {size {sizeRule=Fill, value=1}}}
  }
  CommonTextBlock_65 {
   widgetClassPath=&CommonTextBlock
   parent=HorizontalBox_181
   widget=CommonTextBlock_65
   properties {
    style=&TextStyle_RobotoEngineLike_C
    text=Blur
    colorAndOpacity {colorUseRule=UseColor_Specified, specifiedColor=<rgba 0.5 0.5 0.5 1>}
    font {
     bForceMonospaced=false
     bMaterialIsStencil=false
     fontMaterial=null
     fontObject=&Roboto
     letterSpacing=0
     monospacedWidth=1
     size=8
     skewAmount=0
     typefaceFontName=Regular
     outlineSettings {
      bApplyOutlineToDropShadows=false
      bMiteredCorners=false
      bSeparateFillAlpha=false
      outlineMaterial=null
      outlineSize=0
      outlineColor=<rgba 0 0 0 1>
     }
    }
    shadowOffset=<0 0>
   }
   slot {
    class=HorizontalBoxSlot
    properties {horizontalAlignment=HAlign_Left, verticalAlignment=VAlign_Center, padding=<ltrb 5 0 0 0>, size {sizeRule=Fill, value=0.15}}
   }
  }
  SinglePropertyView {
   widgetClassPath=&SinglePropertyView
   parent=HorizontalBox_181
   widget=SinglePropertyView
   properties {nameOverride=" ", propertyName=Blur}
   slot {class=HorizontalBoxSlot, properties {size {sizeRule=Fill, value=1}}}
  }
  Button {
   widgetClassPath=&Button
   parent=StackBox_0
   widget=Button
   properties {widgetStyle {
     clickedSlateSound {resourceObject=null}
     disabled {
      bIsDynamicallyLoaded=false
      drawAs=RoundedBox
      imageType=NoImage
      mirroring=NoMirror
      resourceName=None
      resourceObject=null
      tiling=NoTile
      imageSize=<0 0>
      margin=<ltrb 0 0 0 0>
      outlineSettings {
       bUseBrushTransparency=false
       roundingType=FixedRadius
       width=1
       color {colorUseRule=UseColor_ColorTable, specifiedColor=<rgba 0 0 0 0>}
       cornerRadii=<4 4 4 4>
      }
      tintColor {colorUseRule=UseColor_Specified, specifiedColor=<rgba 0.0395 0.0395 0.0395 1>}
      uVRegion {bIsValid=false, max=<0 0>, min=<0 0>}
     }
     disabledForeground {colorUseRule=UseColor_ColorTable, specifiedColor=<rgba 0 0 0 0>}
     hovered {
      bIsDynamicallyLoaded=false
      drawAs=RoundedBox
      imageType=NoImage
      mirroring=NoMirror
      resourceName=None
      resourceObject=null
      tiling=NoTile
      imageSize=<32 32>
      margin=<ltrb 0 0 0 0>
      outlineSettings {
       bUseBrushTransparency=false
       roundingType=FixedRadius
       width=1
       color {colorUseRule=UseColor_ColorTable, specifiedColor=<rgba 0 0 0 0>}
       cornerRadii=<4 4 4 4>
      }
      tintColor {colorUseRule=UseColor_Specified, specifiedColor=<rgba 0.0953 0.0953 0.0953 1>}
      uVRegion {bIsValid=false, max=<0 0>, min=<0 0>}
     }
     hoveredForeground {colorUseRule=UseColor_ColorTable, specifiedColor=<rgba 0 0 0 0>}
     hoveredSlateSound {resourceObject=null}
     normal {
      bIsDynamicallyLoaded=false
      drawAs=RoundedBox
      imageType=NoImage
      mirroring=NoMirror
      resourceName=None
      resourceObject=null
      tiling=NoTile
      imageSize=<32 32>
      margin=<ltrb 0 0 0 0>
      outlineSettings {
       bUseBrushTransparency=false
       roundingType=FixedRadius
       width=1
       color {colorUseRule=UseColor_ColorTable, specifiedColor=<rgba 0 0 0 0>}
       cornerRadii=<4 4 4 4>
      }
      tintColor {colorUseRule=UseColor_Specified, specifiedColor=<rgba 0.0395 0.0395 0.0395 1>}
      uVRegion {bIsValid=false, max=<0 0>, min=<0 0>}
     }
     normalForeground {colorUseRule=UseColor_ColorTable, specifiedColor=<rgba 0 0 0 0>}
     normalPadding=<ltrb 12 1.5 12 1.5>
     pressed {
      bIsDynamicallyLoaded=false
      drawAs=RoundedBox
      imageType=NoImage
      mirroring=NoMirror
      resourceName=None
      resourceObject=null
      tiling=NoTile
      imageSize=<32 32>
      margin=<ltrb 0 0 0 0>
      outlineSettings {
       bUseBrushTransparency=false
       roundingType=FixedRadius
       width=1
       color {colorUseRule=UseColor_ColorTable, specifiedColor=<rgba 0 0 0 0>}
       cornerRadii=<4 4 4 4>
      }
      tintColor {colorUseRule=UseColor_Specified, specifiedColor=<rgba 0.0284 0.0284 0.0284 1>}
      uVRegion {bIsValid=false, max=<0 0>, min=<0 0>}
     }
     pressedForeground {colorUseRule=UseColor_ColorTable, specifiedColor=<rgba 0 0 0 0>}
     pressedPadding=<ltrb 12 2.5 12 0.5>
     pressedSlateSound {resourceObject=null}
    }}
   slot {class=StackBoxSlot, properties {padding=<ltrb 5 5 5 5>}}
  }
  CommonTextBlock {
   widgetClassPath=&CommonTextBlock
   parent=Button
   widget=CommonTextBlock
   properties {
    style=&TextStyle_RobotoEngineLike_C
    text=New Brush
    colorAndOpacity {colorUseRule=UseColor_Specified, specifiedColor=<rgba 0.5 0.5 0.5 1>}
    font {
     bForceMonospaced=false
     bMaterialIsStencil=false
     fontMaterial=null
     fontObject=&Roboto
     letterSpacing=0
     monospacedWidth=1
     size=8
     skewAmount=0
     typefaceFontName=Regular
     outlineSettings {
      bApplyOutlineToDropShadows=false
      bMiteredCorners=false
      bSeparateFillAlpha=false
      outlineMaterial=null
      outlineSize=0
      outlineColor=<rgba 0 0 0 1>
     }
    }
    shadowOffset=<0 0>
   }
   slot {class=ButtonSlot}
  }
  HorizontalBox_176 {
   widgetClassPath=&HorizontalBox
   parent=StackBox_0
   widget=HorizontalBox_176
   slot {class=StackBoxSlot, properties {padding=<ltrb 5 5 5 5>}}
  }
  HorizontalBox_0 {
   widgetClassPath=&HorizontalBox
   parent=HorizontalBox_176
   widget=HorizontalBox_0
   slot {class=HorizontalBoxSlot, properties {size {sizeRule=Fill, value=1}}}
  }
  EditableText_48 {
   widgetClassPath=/Script/UMG.EditableText
   bIsVariable=true
   parent=HorizontalBox_0
   widget=EditableText_48
   properties {widgetStyle {
     backgroundImageComposing {
      bIsDynamicallyLoaded=false
      drawAs=Border
      imageType=FullColor
      mirroring=NoMirror
      resourceName=../../../Engine/Content/Slate/Old/HyperlinkDotted.png
      resourceObject=null
      tiling=Both
      imageSize=<0 0>
      margin=<ltrb 0 0 0 0.1875>
      outlineSettings {
       bUseBrushTransparency=false
       roundingType=HalfHeightRadius
       width=0
       color {colorUseRule=UseColor_Specified, specifiedColor=<rgba 0 0 0 0>}
       cornerRadii=<0 0 0 0>
      }
      tintColor {colorUseRule=UseColor_Specified, specifiedColor=<rgba 1 1 1 1>}
      uVRegion {bIsValid=false, max=<0 0>, min=<0 0>}
     }
     backgroundImageSelected {
      bIsDynamicallyLoaded=false
      drawAs=Image
      imageType=NoImage
      mirroring=NoMirror
      resourceName=None
      resourceObject=null
      tiling=NoTile
      imageSize=<0 0>
      margin=<ltrb 0 0 0 0>
      outlineSettings {
       bUseBrushTransparency=false
       roundingType=HalfHeightRadius
       width=0
       color {colorUseRule=UseColor_Specified, specifiedColor=<rgba 0 0 0 0>}
       cornerRadii=<0 0 0 0>
      }
      tintColor {colorUseRule=UseColor_Specified, specifiedColor=<rgba 0 0.162 0.7454 1>}
      uVRegion {bIsValid=false, max=<0 0>, min=<0 0>}
     }
     caretImage {
      bIsDynamicallyLoaded=false
      drawAs=Image
      imageType=NoImage
      mirroring=NoMirror
      resourceName=None
      resourceObject=null
      tiling=NoTile
      imageSize=<0 0>
      margin=<ltrb 0 0 0 0>
      outlineSettings {
       bUseBrushTransparency=false
       roundingType=HalfHeightRadius
       width=0
       color {colorUseRule=UseColor_Specified, specifiedColor=<rgba 0 0 0 0>}
       cornerRadii=<0 0 0 0>
      }
      tintColor {colorUseRule=UseColor_Specified, specifiedColor=<rgba 1 1 1 1>}
      uVRegion {bIsValid=false, max=<0 0>, min=<0 0>}
     }
     colorAndOpacity {colorUseRule=UseColor_Foreground, specifiedColor=<rgba 1 0 1 1>}
     font {
      bForceMonospaced=false
      bMaterialIsStencil=false
      fontMaterial=null
      fontObject=null
      letterSpacing=0
      monospacedWidth=1
      size=9
      skewAmount=0
      typefaceFontName=Regular
      outlineSettings {
       bApplyOutlineToDropShadows=false
       bMiteredCorners=false
       bSeparateFillAlpha=false
       outlineMaterial=null
       outlineSize=0
       outlineColor=<rgba 0 0 0 1>
      }
     }
     hintColorAndOpacity {colorUseRule=UseColor_Foreground, specifiedColor=<rgba 1 0 1 1>}
     selectedBackgroundColor {colorUseRule=UseColor_Foreground, specifiedColor=<rgba 1 0 1 1>}
    }}
   slot {class=HorizontalBoxSlot, properties {size {sizeRule=Fill, value=1}}}
  }
  HorizontalBox_92 {widgetClassPath=&HorizontalBox, parent=HorizontalBox_176, widget=HorizontalBox_92, slot {class=HorizontalBoxSlot}}
  Button_130 {
   widgetClassPath=&Button
   parent=HorizontalBox_92
   widget=Button_130
   properties {widgetStyle {
     clickedSlateSound {resourceObject=null}
     disabled {
      bIsDynamicallyLoaded=false
      drawAs=RoundedBox
      imageType=NoImage
      mirroring=NoMirror
      resourceName=None
      resourceObject=null
      tiling=NoTile
      imageSize=<0 0>
      margin=<ltrb 0 0 0 0>
      outlineSettings {
       bUseBrushTransparency=false
       roundingType=FixedRadius
       width=1
       color {colorUseRule=UseColor_ColorTable, specifiedColor=<rgba 0 0 0 0>}
       cornerRadii=<4 4 4 4>
      }
      tintColor {colorUseRule=UseColor_Specified, specifiedColor=<rgba 0.0395 0.0395 0.0395 1>}
      uVRegion {bIsValid=false, max=<0 0>, min=<0 0>}
     }
     disabledForeground {colorUseRule=UseColor_ColorTable, specifiedColor=<rgba 0 0 0 0>}
     hovered {
      bIsDynamicallyLoaded=false
      drawAs=RoundedBox
      imageType=NoImage
      mirroring=NoMirror
      resourceName=None
      resourceObject=null
      tiling=NoTile
      imageSize=<32 32>
      margin=<ltrb 0 0 0 0>
      outlineSettings {
       bUseBrushTransparency=false
       roundingType=FixedRadius
       width=1
       color {colorUseRule=UseColor_ColorTable, specifiedColor=<rgba 0 0 0 0>}
       cornerRadii=<4 4 4 4>
      }
      tintColor {colorUseRule=UseColor_Specified, specifiedColor=<rgba 0.0953 0.0953 0.0953 1>}
      uVRegion {bIsValid=false, max=<0 0>, min=<0 0>}
     }
     hoveredForeground {colorUseRule=UseColor_ColorTable, specifiedColor=<rgba 0 0 0 0>}
     hoveredSlateSound {resourceObject=null}
     normal {
      bIsDynamicallyLoaded=false
      drawAs=RoundedBox
      imageType=NoImage
      mirroring=NoMirror
      resourceName=None
      resourceObject=null
      tiling=NoTile
      imageSize=<32 32>
      margin=<ltrb 0 0 0 0>
      outlineSettings {
       bUseBrushTransparency=false
       roundingType=FixedRadius
       width=1
       color {colorUseRule=UseColor_ColorTable, specifiedColor=<rgba 0 0 0 0>}
       cornerRadii=<4 4 4 4>
      }
      tintColor {colorUseRule=UseColor_Specified, specifiedColor=<rgba 0.0395 0.0395 0.0395 1>}
      uVRegion {bIsValid=false, max=<0 0>, min=<0 0>}
     }
     normalForeground {colorUseRule=UseColor_ColorTable, specifiedColor=<rgba 0 0 0 0>}
     normalPadding=<ltrb 12 1.5 12 1.5>
     pressed {
      bIsDynamicallyLoaded=false
      drawAs=RoundedBox
      imageType=NoImage
      mirroring=NoMirror
      resourceName=None
      resourceObject=null
      tiling=NoTile
      imageSize=<32 32>
      margin=<ltrb 0 0 0 0>
      outlineSettings {
       bUseBrushTransparency=false
       roundingType=FixedRadius
       width=1
       color {colorUseRule=UseColor_ColorTable, specifiedColor=<rgba 0 0 0 0>}
       cornerRadii=<4 4 4 4>
      }
      tintColor {colorUseRule=UseColor_Specified, specifiedColor=<rgba 0.0284 0.0284 0.0284 1>}
      uVRegion {bIsValid=false, max=<0 0>, min=<0 0>}
     }
     pressedForeground {colorUseRule=UseColor_ColorTable, specifiedColor=<rgba 0 0 0 0>}
     pressedPadding=<ltrb 12 2.5 12 0.5>
     pressedSlateSound {resourceObject=null}
    }}
   slot {class=HorizontalBoxSlot}
  }
  CommonTextBlock_30 {
   widgetClassPath=&CommonTextBlock
   parent=Button_130
   widget=CommonTextBlock_30
   properties {
    style=&TextStyle_RobotoEngineLike_C
    text=Save
    colorAndOpacity {colorUseRule=UseColor_Specified, specifiedColor=<rgba 0.5 0.5 0.5 1>}
    font {
     bForceMonospaced=false
     bMaterialIsStencil=false
     fontMaterial=null
     fontObject=&Roboto
     letterSpacing=0
     monospacedWidth=1
     size=8
     skewAmount=0
     typefaceFontName=Regular
     outlineSettings {
      bApplyOutlineToDropShadows=false
      bMiteredCorners=false
      bSeparateFillAlpha=false
      outlineMaterial=null
      outlineSize=0
      outlineColor=<rgba 0 0 0 1>
     }
    }
    shadowOffset=<0 0>
   }
   slot {class=ButtonSlot}
  }
  Spacer_66 {widgetClassPath=&Spacer, parent=StackBox_0, widget=Spacer_66, slot {class=StackBoxSlot, properties {padding=<ltrb 0 0 0 40>}}}
  DetailsView_39 {
   widgetClassPath=&DetailsView
   parent=StackBox_0
   widget=DetailsView_39
   properties {propertiesToShow [Actors,Bounds]}
   slot {class=StackBoxSlot}
  }
  Image_0 {
   widgetClassPath=&Image
   bIsVariable=true
   parent=CanvasPanel_18
   widget=Image_0
   properties {brush {
     bIsDynamicallyLoaded=false
     drawAs=Image
     imageType=NoImage
     mirroring=NoMirror
     resourceName=None
     resourceObject=/Game/ArtTools/Tech_Art/FogDensityBaker/Materials/M_PreviewWindow
     tiling=NoTile
     imageSize=<32 32>
     margin=<ltrb 0 0 0 0>
     outlineSettings {
      bUseBrushTransparency=false
      roundingType=HalfHeightRadius
      width=0
      color {colorUseRule=UseColor_Specified, specifiedColor=<rgba 0 0 0 0>}
      cornerRadii=<0 0 0 0>
     }
     tintColor {colorUseRule=UseColor_Specified, specifiedColor=<rgba 1 1 1 1>}
     uVRegion {bIsValid=false, max=<0 0>, min=<0 0>}
    }}
   slot {
    class=CanvasPanelSlot
    properties {bAutoSize=true, layoutData {alignment=<0 0>, anchors {maximum=<0 0>, minimum=<0 0>}, offsets=<ltrb 360 0 100 30>}}
   }
  }
 ]
}
