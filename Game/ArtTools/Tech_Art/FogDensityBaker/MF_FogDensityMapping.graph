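# Material function that converts a world-space density map into a fog density value, part of the FogDensityBaker tech-art tool.
# World XY is normalized between Min and Max Bounds, then inverted on both axes to form UVs for sampling the Density Map, which defaults to a black mask.
# The sample drives a three-point lerp from Min Density through a neutral 1 to Max Density.
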MaterialFunction /Game/ArtTools/Tech_Art/FogDensityBaker/MF_FogDensityMapping
graph {nodes [
  n1 = FunctionInput() {bUsePreviewValueAsDefault=true, inputName=Max Density, inputType=FunctionInput_Scalar, sortPriority=30, previewValue=<2 0 0 1>}
  n2 = WorldPosition()
  n3 = FunctionInput() {
   bUsePreviewValueAsDefault=true
   inputName=Min Bounds
   inputType=FunctionInput_Vector2
   sortPriority=6
   previewValue=<-34583.8008 -96628.9766 0 1>
  }
  n4 = FunctionInput() {
   bUsePreviewValueAsDefault=true
   inputName=Max Bounds
   inputType=FunctionInput_Vector2
   sortPriority=9
   previewValue=<110129.4141 48084.2344 0 1>
  }
  n5 = OneMinus(None=Divide(A=Subtract(A=@n2.XY, B=@n3), B=Subtract(A=@n4, B=@n3)))
  n6 = FunctionInput(Preview=TextureObject() {samplerType=SAMPLERTYPE_Masks, texture=&T_BlackMask}) {bUsePreviewValueAsDefault=true, inputName=Density Map, inputType=FunctionInput_Texture2D}
  FunctionOutput(None=MaterialFunctionCall!(
    1=FunctionInput() {bUsePreviewValueAsDefault=true, inputName=Min Density, inputType=FunctionInput_Scalar, sortPriority=15}
    2=Constant() {r=1}
    3=@n1
    4
    A=TextureSample(UVs=@n5, Tex=@n6).RGB
   ) {materialFunction=/Engine/Functions/Engine_MaterialFunctions02/Lerp_Multiple_Float}.Lerp 3 Inputs)
 ]}
