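# Test material in the fog density baker tech-art tools that exercises the MF_FogDensityMapping function.
# A Bounds vector parameter supplies Min Bounds from its red channel and Max Bounds from blue, each duplicated into both float2 components, leaving green and alpha unused.
# Density Map and min/max density inputs are unconnected, and no material output is wired.
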Material /Game/ArtTools/Tech_Art/FogDensityBaker/Materials/M_MappingTest
graph {
 properties {pixelDepthOffsetMode=PDOM_Legacy, usedShadingModels=Default Lit}
 nodes [
  n1 = VectorParameter() {parameterName=Bounds, defaultValue=<rgba -34583.8008 -96628.9766 110129.4141 48084.2344>}
  MaterialFunctionCall!(
   Density Map
   Min Bounds=MaterialFunctionCall(X=@n1.R, Y=@n1.R) {materialFunction=&MakeFloat2}.Result
   Max Bounds=MaterialFunctionCall(X=@n1.B, Y=@n1.B) {materialFunction=&MakeFloat2}.Result
   Min Density
   Max Density
  ) {materialFunction=/Game/ArtTools/Tech_Art/FogDensityBaker/MF_FogDensityMapping}
 ]
}
