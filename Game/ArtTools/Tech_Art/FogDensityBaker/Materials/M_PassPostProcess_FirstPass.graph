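# Post-process material for the first pass of the FogDensityBaker art tool, deriving a single-channel value from scene BaseColor and depth.
# It takes the red channel of BaseColor and replaces it with mid-grey wherever SceneDepth exceeds a far cutoff, using an If-driven lerp.
# A feature level switch returns zero for ES2 and ES3_1.
# A separate fractional-depth branch, which would produce repeating bands, feeds no other node shown.
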Material /Game/ArtTools/Tech_Art/FogDensityBaker/Materials/M_PassPostProcess_FirstPass
graph {
 properties {materialDomain=MD_PostProcess, pixelDepthOffsetMode=PDOM_Legacy, usedShadingModels=Default Lit}
 nodes [
  n1 = ComponentMask(None=SceneTexture() {sceneTextureId=PPI_BaseColor}.Color) {b=true, g=true, r=true}
  n2 = SceneDepth()
  n3 = Constant() {r=200000}
  n4 = Constant() {r=1}
  n5 = Constant()
  n6 = Constant()
  FeatureLevelSwitch(
   Default=LinearInterpolate(A=ComponentMask(None=@n1) {r=true}, Alpha=If(A=@n2.R, B=@n3, "A > B"=@n4, "A == B"=@n5, "A < B"=@n5)) {constB=0.5}
   ES2=@n6
   ES3_1=@n6
  )
  Frac(None=Divide(A=@n2.R) {constB=100})
 ]
}
